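# Default Lit material in the Athena Ice environment, named for matching snow terrain.
# It uses material attributes and calls the landscape snow function MF_Athena_Snow_01, which exposes a UseLandscapeCoordsOnSwitch input.
# The function's Material output feeds a GetMaterialAttributes node.
# A VertexColor node is also present, with no connection shown from it or into the final material output.
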Material /FigmentEnvironment/LM_8_1_1/Athena/Environments/Ice/Materials/M_Athena_SnowTerrainMatch
graph {
 properties {bUseMaterialAttributes=true, pixelDepthOffsetMode=PDOM_Legacy, usedShadingModels=Default Lit}
 nodes [
  GetMaterialAttributes(None=MaterialFunctionCall!(UseLandscapeCoordsOnSwitch) {materialFunction=/FigmentEnvironment/LM_8_1_1/Athena/Environments/Landscape/MaterialFunctions/Snow/MF_Athena_Snow_01}.Material)
  VertexColor()
 ]
}
